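# Translucent UI material, named as a backplate for social party and game buttons.
# It pairs a constant black color with a faint scalar value that is either flat or a vertical gradient.
# A floored Show Fade parameter acts as a switch, swapping the flat value for a ramp along normalized UV Y from GetUserInterfaceUV.
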Material /FortniteCoreUI/Social/Buttons/Materials/M_Social_Backplate_PartyAndGame
graph {
 properties {bAutomaticallySetUsageInEditor=false, blendMode=BLEND_Translucent, materialDomain=MD_UI, usedShadingModels=Default Lit}
 nodes [
  Constant3Vector()
  n1 = MaterialFunctionCall() {materialFunction=&GetUserInterfaceUV}
  LinearInterpolate(
   B=LinearInterpolate(Alpha=ComponentMask(None=@n1.Normalized UV) {g=true}) {constB=0.15}
   Alpha=Floor(None=ScalarParameter() {parameterName=Show Fade, sliderMax=1})
  ) {constA=0.15}
 ]
}
