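# M_Social_Divider is a translucent UI material, named and filed as a divider for the social button UI.
# It pairs a constant muted lavender-grey color with a procedural falloff, using no textures or parameters.
# A reflected gradient drives a lerp between two low values, and the result is shaped by the shared ease-curves material function.
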Material /FortniteCoreUI/Social/Buttons/Materials/M_Social_Divider
graph {
 properties {bAutomaticallySetUsageInEditor=false, blendMode=BLEND_Translucent, materialDomain=MD_UI, usedShadingModels=Default Lit}
 nodes [
  Constant3Vector() {constant=<rgba 0.3864 0.3763 0.4793 1>}
  MaterialFunctionCall("Value (0-1)"=LinearInterpolate(Alpha=MaterialFunctionCall!(UVs, Offset, RotationCenter, RotationAngle) {materialFunction=/FortUILibrary/MaterialFunctions/Gradients/MF_UI_ReflectedGradient}.Result) {constA=0.1, constB=0.35}) {materialFunction=&MF_UI_EaseCurves}
 ]
}
