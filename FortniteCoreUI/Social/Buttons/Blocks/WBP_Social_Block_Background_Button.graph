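# Background block widget for social UI buttons, layering a backplate, a named content slot and a focus outline.
# State functions pick the backplate color from the social button type, with one type also checking IsSelected on hover and unhover.
# Hover also focuses the outline and sets its material's SheenRotation from the last navigation direction, while unhover clears focus instantly.
# Helper functions apply padding containment and corner radius to both backplate and outline.
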WidgetBlueprint /FortniteCoreUI/Social/Buttons/Blocks/WBP_Social_Block_Background_Button
graphs [
 BackplateChangeOnDeselected {nodes [
   n1 = BackplateChangeOnDeselected!() -> (then, Button Type:E_UI_Social_Button_Type Enum=NewEnumerator0) {is=title K2Node_FunctionEntry}
   > SetBackplateColor!(
    execute
    self:WBP Social Block Backplate Object Reference=Get Backplate() -> (Backplate:WBP Social Block Backplate Object Reference)
    Backplate Switch:E_UI_Social_Button_Backplate_Color Enum=Select!(
     NewEnumerator0:E_UI_Social_Button_Backplate_Color Enum=NewEnumerator1
     NewEnumerator1:E_UI_Social_Button_Backplate_Color Enum=NewEnumerator1
     NewEnumerator2:E_UI_Social_Button_Backplate_Color Enum=NewEnumerator13
     NewEnumerator3:E_UI_Social_Button_Backplate_Color Enum=NewEnumerator0
     NewEnumerator4:E_UI_Social_Button_Backplate_Color Enum=NewEnumerator0
     NewEnumerator6:E_UI_Social_Button_Backplate_Color Enum=NewEnumerator12
     NewEnumerator7:E_UI_Social_Button_Backplate_Color Enum=NewEnumerator0
     NewEnumerator8:E_UI_Social_Button_Backplate_Color Enum=NewEnumerator12
     Index:E_UI_Social_Button_Type Enum=@n1
    ) -> (ReturnValue:E_UI_Social_Button_Backplate_Color Enum)
   ) -> (then)
  ]}
 BackplateChangeOnDisabled {nodes [
   n1 = BackplateChangeOnDisabled!() -> (then, Button Type:E_UI_Social_Button_Type Enum=NewEnumerator0) {is=title K2Node_FunctionEntry}
   > SetBackplateColor(
    self=Get Backplate()
    Backplate Switch=Select(
     NewEnumerator0=NewEnumerator1
     NewEnumerator1=NewEnumerator1
     NewEnumerator2=NewEnumerator1
     NewEnumerator3=NewEnumerator0
     NewEnumerator4=NewEnumerator0
     NewEnumerator6=NewEnumerator12
     NewEnumerator7=NewEnumerator0
     NewEnumerator8=NewEnumerator12
     Index=@n1
    )
   )
  ]}
 BackplateChangeOnHover {nodes [
   BackplateChangeOnHover!() -> (then, Button Type:E_UI_Social_Button_Type Enum=NewEnumerator0) {is=title K2Node_FunctionEntry}
   n1 = Get IsSelected() -> (IsSelected:bool)
   > SetBackplateColor(
    self=Get Backplate()
    Backplate Switch=Select(
     NewEnumerator0=NewEnumerator2
     NewEnumerator1=NewEnumerator4
     NewEnumerator2=NewEnumerator13
     NewEnumerator3=NewEnumerator2
     NewEnumerator4=NewEnumerator0
     NewEnumerator6=NewEnumerator12
     NewEnumerator7=Select!(
      Option 0:E_UI_Social_Button_Backplate_Color Enum=NewEnumerator0
      Option 1:E_UI_Social_Button_Backplate_Color Enum=NewEnumerator13
      Index:bool=@n1
     ) -> (ReturnValue)
     NewEnumerator8=NewEnumerator4
     Index=Get Button Type() -> (Button Type:E_UI_Social_Button_Type Enum=NewEnumerator0)
    )
   )
   > Play Shimmer Animation!(execute:Exec) -> (then:Exec) {is=title K2Node_MacroInstance}
   > RequestTransition!(
    execute
    self:UIKit Block Object Reference=Get FocusOutline() -> (FocusOutline:WBP Social Block Outline Object Reference)
    State:EUIKitBlockVisualState Enum=Focused
   ) -> (then)
  ]}
 BackplateChangeOnSelected {nodes [
   n1 = BackplateChangeOnSelected!() -> (then, Button Type:E_UI_Social_Button_Type Enum=NewEnumerator0) {is=title K2Node_FunctionEntry}
   > SetBackplateColor(
    self=Get Backplate()
    Backplate Switch=Select!(
     NewEnumerator0:E_UI_Social_Button_Backplate_Color Enum=NewEnumerator1
     NewEnumerator1:E_UI_Social_Button_Backplate_Color Enum=NewEnumerator1
     NewEnumerator2:E_UI_Social_Button_Backplate_Color Enum=NewEnumerator13
     NewEnumerator3:E_UI_Social_Button_Backplate_Color Enum=NewEnumerator0
     NewEnumerator4:E_UI_Social_Button_Backplate_Color Enum=NewEnumerator0
     NewEnumerator6:E_UI_Social_Button_Backplate_Color Enum=NewEnumerator12
     NewEnumerator7:E_UI_Social_Button_Backplate_Color Enum=NewEnumerator13
     NewEnumerator8:E_UI_Social_Button_Backplate_Color Enum=NewEnumerator12
     Index:E_UI_Social_Button_Type Enum=@n1
    ) -> (ReturnValue)
   )
  ]}
 BackplateChangeOnUnhover {nodes [
   BackplateChangeOnUnhover!() -> (then, Button Type:E_UI_Social_Button_Type Enum=NewEnumerator0) {is=title K2Node_FunctionEntry}
   n1 = Get IsSelected()
   > SetBackplateColor(
    self=Get Backplate()
    Backplate Switch=Select(
     NewEnumerator0=NewEnumerator1
     NewEnumerator1=NewEnumerator1
     NewEnumerator2=NewEnumerator13
     NewEnumerator3=NewEnumerator0
     NewEnumerator4=NewEnumerator0
     NewEnumerator6=NewEnumerator12
     NewEnumerator7=Select!(
      Option 0:E_UI_Social_Button_Backplate_Color Enum=NewEnumerator0
      Option 1:E_UI_Social_Button_Backplate_Color Enum=NewEnumerator13
      Index:bool=@n1
     ) -> (ReturnValue)
     NewEnumerator8=NewEnumerator12
     Index=Get Button Type() -> (Button Type=NewEnumerator0)
    )
   )
   > RequestInstantTransition!(execute, self:UIKit Block Object Reference=Get FocusOutline(), State:EUIKitBlockInstantTransitionState Enum=InstantUnfocused) -> (then)
  ]}
 ContainBackground {nodes [
   n1 = ContainBackground!() -> (then, Contain Background:bool=~) {is=title K2Node_FunctionEntry}
   > n2 = Set ContainBackgroundPadding(ContainBackgroundPadding:bool=@n1) -> (Output_Get:bool)
   > ContainBackplate!(execute, self:WBP Social Block Backplate Object Reference=Get Backplate(), ContainBackplatePadding:bool=@n2) -> (then)
   > ContainOutline!(execute, self:WBP Social Block Outline Object Reference=Get FocusOutline(), ContainOutlinePadding:bool=@n2) -> (then)
  ]}
 EventGraph {nodes [
   n1 = Event PreConstruct()
   > Branch(Condition=@n1.IsDesignTime)
   n2 = Get Slot_BetweenContent() -> (Slot_BetweenContent:Named Slot Object Reference)
   > SetVisibility(
    self=@n2
    InVisibility=Select(
     Option 0:ESlateVisibility Enum=Collapsed
     Option 1:ESlateVisibility Enum=SelfHitTestInvisible
     Index=HasAnyChildren!(self:Panel Widget Object Reference=@n2) -> (ReturnValue:bool)
    ) -> (ReturnValue:ESlateVisibility Enum)
   )
  ]}
 PlayShimmerAnimation {nodes [
   Inputs()
   > n1 = SetShimmerDirection!(execute, self, LastNavDirection:EUINavigation Enum=Get LastNavigationDirection() -> (LastNavigationDirection:EUINavigation Enum)) -> (then, ShimmerRotation:double)
   > n2 = GetDynamicMaterial(self=Get FocusOutline!(self:WBP Social Block Outline Object Reference=Get FocusOutline!() -> (FocusOutline:WBP Social Block Outline Object Reference)) -> (FocusOutline:Image Object Reference))
   > SetScalarParameterValue(self=@n2, ParameterName=SheenRotation, Value=@n1)
   > Outputs!(then:Exec)
  ]}
 SetCornerRadius {nodes [
   SetCornerRadius!() -> (then, Radius In Pixels:bool=~, Corner Radius:double=~) {is=title K2Node_FunctionEntry}
   > SetCornerRadius!(
    execute
    self:WBP Social Block Backplate Object Reference=Get Backplate()
    Radius In Pixels:bool=Get Radius In Pixels() -> (Radius In Pixels:bool)
    Corner Radius:double=Get Corner Radius() -> (Corner Radius:double)
   ) -> (then) {is=title K2Node_CallFunction}
   n1 = Get Corner Radius()
   n2 = Select(Option 0:float=@n1, Option 1:float=float + float(A=@n1, B=13), Index=Get Radius In Pixels()) -> (ReturnValue:float)
   > SetCornerRadiuses!(
    execute
    self:WBP Social Block Outline Object Reference=Get FocusOutline()
    TopLeft:float=@n2
    TopRight:float=@n2
    BottomLeft:float=@n2
    BottomRight:float=@n2
    RadiusInPixels:bool=Get Radius In Pixels()
   ) -> (then)
  ]}
 SetShimmerDirection {nodes [
   n1 = SetShimmerDirection!() -> (then, LastNavDirection:EUINavigation Enum) {is=title K2Node_FunctionEntry}
   > n2 = Switch on EUINavigation!(execute, Selection:EUINavigation Enum=@n1) -> (Left:Exec, Right:Exec, Up:Exec, Down:Exec, Next:Exec, Previous:Exec, Invalid:Exec) {is=title K2Node_SwitchEnum}
   n3 = Set ShimmerRotationValue(execute=@n2.Down, ShimmerRotationValue:double=0.75) -> (Output_Get:double)
   > Return Node!(execute, ShimmerRotation:double=@n3)
   n4 = Set ShimmerRotationValue(execute=[@n2.Invalid, @n2.Right], ShimmerRotationValue=0)
   > Return Node(ShimmerRotation=@n4)
   n5 = Set ShimmerRotationValue(execute=@n2.Left, ShimmerRotationValue=0.5)
   > Return Node(ShimmerRotation=@n5)
   n6 = Set ShimmerRotationValue(execute=@n2.Up, ShimmerRotationValue=0.25)
   > Return Node(ShimmerRotation=@n6)
  ]}
]
variables [
 ContainBackgroundPadding {type=boolean}
 IsSelected {type=boolean}
 LastNavigationDirection {type=EUINavigation, default=Invalid}
 ShimmerRotationValue {type=number}
]
widgets {
 info {inheritedWidgetCount=0, parentClass=&CommonUserWidget, rootWidgetClass=&GridPanel, widgetCount=5}
 widgets [
  Grid_Root {widgetClassPath=&GridPanel, widget=Grid_Root, properties {columnFill [1], rowFill [1]}}
  Backplate_PartyAndGame {
   widgetClassPath=&/WBP_Social_PartyAndGame_Backplate
   bIsVariable=true
   parent=Grid_Root
   widget=Backplate_PartyAndGame
   properties {showFade=false, visibility=Collapsed}
   slot {class=GridSlot, properties {padding=<ltrb -18 0 -22 0>}}
  }
  Backplate {
   widgetClassPath=&/WBP_Social_Block_Backplate
   bIsVariable=true
   parent=Grid_Root
   widget=Backplate
   slot {class=GridSlot}
  }
  Slot_BetweenContent {
   widgetClassPath=&NamedSlot
   bIsVariable=true
   parent=Grid_Root
   widget=Slot_BetweenContent
   slot {class=GridSlot, properties {layer=1}}
  }
  FocusOutline {
   widgetClassPath=&/WBP_Social_Block_Outline
   bIsVariable=true
   parent=Grid_Root
   widget=FocusOutline
   properties {bottomLeftCornerRadius=10, bottomRightCornerRadius=10, radiusInPixels=true, topLeftCornerRadius=10, topRightCornerRadius=10}
   slot {class=GridSlot, properties {layer=99}}
  }
 ]
}
