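# Generic social UI button built on the UIKit modular button base, styled by a button type enum.
# Hover, select and enable events drive the ButtonBackground block's backplate and font materials; disabled also dims opacity.
# Font material selection varies by type and touch input, and input method changes re-initialize visuals.
# PreConstruct registers the focus outline and initializes visuals, then swaps in the compact-UI corner radius variable without reapplying it.
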WidgetBlueprint /FortniteCoreUI/Social/Buttons/WBP_Social_ButtonGeneric
graphs [
 EventGraph {nodes [
   n1 = Event PreConstruct()
   > Parent: PreConstruct(IsDesignTime=@n1.IsDesignTime)
   > RegisterModularBlock!(
    execute
    self:UIKit Modular Button Object Reference
    Block:UIKit Block Object Reference=Get FocusOutline!(self:WBP Social Block Background Button Object Reference=Get ButtonBackground() -> (ButtonBackground:WBP Social Block Background Button Object Reference)) -> (FocusOutline:WBP Social Block Outline Object Reference)
   ) -> (then)
   > InitializeButtonVisuals()
   > SetCompactUIStyle() {is=title K2Node_CallFunction}
   Event On Enabled() {is=title K2Node_Event}
   > Parent: On Enabled() {is=title K2Node_CallParentFunction}
   > SetRenderOpacity(self=Self-Reference(), InOpacity=1)
   > n2 = Branch(Condition=IsHovered())
   > BackplateChangeOnHover!(
    execute
    self:WBP Social Block Background Button Object Reference=Get ButtonBackground()
    Button Type:E_UI_Social_Button_Type Enum=Get ButtonType() -> (ButtonType:E_UI_Social_Button_Type Enum=NewEnumerator0)
   ) -> (then)
   > SetFontMaterials!(execute, self, Font Materials:Object Reference=Font Materials on Hover!() -> (ReturnValue:Object Reference) {is=title K2Node_MacroInstance}) -> (then)
   BackplateChangeOnUnhover!(
    execute=@n2.else
    self:WBP Social Block Background Button Object Reference=Get ButtonBackground()
    Button Type:E_UI_Social_Button_Type Enum=Get ButtonType() -> (ButtonType=NewEnumerator0)
   ) -> (then)
   > SetFontMaterials(Font Materials=Font Materials on Unhover!() -> (ReturnValue:Object Reference) {is=title K2Node_MacroInstance})
   Event On Disabled() {is=title K2Node_Event}
   > Parent: On Disabled() {is=title K2Node_CallParentFunction}
   > BackplateChangeOnDisabled!(
    execute
    self:WBP Social Block Background Button Object Reference=Get ButtonBackground()
    Button Type:E_UI_Social_Button_Type Enum=Get ButtonType() -> (ButtonType=NewEnumerator0)
   ) -> (then)
   > SetFontMaterials(Font Materials=/FortniteCoreUI/Social/Buttons/Materials/Fonts/MI_Social_FontColor_Secondary)
   > SetRenderOpacity(self=Self-Reference(), InOpacity=0.25)
   Event On Hovered()
   > Parent: On Hovered() {is=title K2Node_CallParentFunction}
   > BackplateChangeOnHover(self=Get ButtonBackground(), Button Type=Get ButtonType() -> (ButtonType=NewEnumerator0))
   > SetFontMaterials(Font Materials=Font Materials on Hover())
   Event On Unhovered()
   > Parent: On Unhovered() {is=title K2Node_CallParentFunction}
   > BackplateChangeOnUnhover(self=Get ButtonBackground(), Button Type=Get ButtonType() -> (ButtonType=NewEnumerator0))
   > SetFontMaterials(Font Materials=Font Materials on Unhover())
   Event On Selected()
   > Parent: On Selected() {is=title K2Node_CallParentFunction}
   n3 = Get ButtonBackground()
   > Set IsSelected!(execute, IsSelected:bool=true, self:WBP Social Block Background Button Object Reference=@n3) -> (then, Output_Get:bool)
   > BackplateChangeOnSelected!(
    execute
    self:WBP Social Block Background Button Object Reference=@n3
    Button Type:E_UI_Social_Button_Type Enum=Get ButtonType() -> (ButtonType=NewEnumerator0)
   ) -> (then)
   Event On Deselected()
   > Parent: On Deselected() {is=title K2Node_CallParentFunction}
   n4 = Get ButtonBackground()
   > Set IsSelected(IsSelected=false, self=@n4)
   > BackplateChangeOnDeselected!(
    execute
    self:WBP Social Block Background Button Object Reference=@n4
    Button Type:E_UI_Social_Button_Type Enum=Get ButtonType() -> (ButtonType=NewEnumerator0)
   ) -> (then)
   n5 = Event On Input Method Changed!() -> (OutputDelegate, then, CurrentInputType:ECommonInputType Enum) {is=title K2Node_Event}
   > Parent: On Input Method Changed!(execute, CurrentInputType:ECommonInputType Enum=@n5.CurrentInputType) -> (then) {is=title K2Node_CallParentFunction}
   > InitializeButtonVisuals()
  ]}
 FontMaterialsOnHover {nodes [
   Inputs!()
   n1 = IsUsingInputType!(InputType:ECommonInputType Enum=Touch, PlayerController:Player Controller Object Reference=GetOwningPlayer()) -> (ReturnValue:bool)
   Outputs!(ReturnValue:Object Reference=Select!(
     NewEnumerator0:Object Reference=/FortniteCoreUI/Social/Buttons/Materials/Fonts/MI_Social_FontColor_Primary
     NewEnumerator1:Object Reference=SelectMaterialInstance!(
      A:Material Instance Object Reference=/FortniteCoreUI/Social/Buttons/Materials/Fonts/MI_Social_FontColor_Button_Alert_Touch
      B:Material Instance Object Reference=/FortniteCoreUI/Social/Buttons/Materials/Fonts/MI_Social_FontColor_Button_Alert
      bSelectA:bool=@n1
     ) -> (ReturnValue:Material Instance Object Reference)
     NewEnumerator2:Object Reference=/FortniteCoreUI/Social/Buttons/Materials/Fonts/MI_Social_FontColor_Button_Attention
     NewEnumerator3:Object Reference=&MI_Social_FontColor_Primary
     NewEnumerator4:Object Reference=&MI_Social_FontColor_Primary
     NewEnumerator6:Object Reference=&MI_Social_FontColor_Primary
     NewEnumerator7:Object Reference=&MI_Social_FontColor_Primary
     NewEnumerator8:Object Reference=&MI_Social_FontColor_Button_Alert
     Index:E_UI_Social_Button_Type Enum=Get ButtonType() -> (ButtonType=NewEnumerator0)
    ) -> (ReturnValue:Object Reference))
  ]}
 FontMaterialsOnUnhover {nodes [
   Inputs()
   n1 = IsUsingInputType(InputType=Touch, PlayerController=GetOwningPlayer())
   Outputs(ReturnValue=Select(
     NewEnumerator0=&MI_Social_FontColor_Secondary
     NewEnumerator1=SelectMaterialInstance(
      A=&MI_Social_FontColor_Button_Alert_Touch
      B=&MI_Social_FontColor_Secondary
      bSelectA=@n1
     )
     NewEnumerator2=&MI_Social_FontColor_Button_Attention
     NewEnumerator3=&MI_Social_FontColor_Secondary
     NewEnumerator4=&MI_Social_FontColor_Secondary
     NewEnumerator6=&MI_Social_FontColor_Secondary
     NewEnumerator7=&MI_Social_FontColor_Primary
     NewEnumerator8=&MI_Social_FontColor_Button_Alert_Touch
     Index=Get ButtonType() -> (ButtonType=NewEnumerator0)
    ))
  ]}
 InitializeButtonVisuals {nodes [
   InitializeButtonVisuals() {is=title K2Node_FunctionEntry}
   > SetBackplateColor!(
    execute
    self:WBP Social Block Backplate Object Reference=Get Backplate!(self:WBP Social Block Background Button Object Reference=Get ButtonBackground()) -> (Backplate:WBP Social Block Backplate Object Reference)
    Backplate Switch:E_UI_Social_Button_Backplate_Color Enum=Select(
     NewEnumerator0:E_UI_Social_Button_Backplate_Color Enum=NewEnumerator1
     NewEnumerator1:E_UI_Social_Button_Backplate_Color Enum=NewEnumerator1
     NewEnumerator2:E_UI_Social_Button_Backplate_Color Enum=NewEnumerator13
     NewEnumerator3:E_UI_Social_Button_Backplate_Color Enum=NewEnumerator0
     NewEnumerator4:E_UI_Social_Button_Backplate_Color Enum=NewEnumerator0
     NewEnumerator6:E_UI_Social_Button_Backplate_Color Enum=NewEnumerator12
     NewEnumerator7:E_UI_Social_Button_Backplate_Color Enum=NewEnumerator0
     NewEnumerator8:E_UI_Social_Button_Backplate_Color Enum=NewEnumerator12
     Index=Get ButtonType() -> (ButtonType=NewEnumerator0)
    ) -> (ReturnValue:E_UI_Social_Button_Backplate_Color Enum)
   ) -> (then)
   > SetCornerRadius!(
    execute
    self:WBP Social Block Background Button Object Reference=Get ButtonBackground()
    Radius In Pixels:bool=Get RadiusInPixels() -> (RadiusInPixels:bool)
    Corner Radius:double=Get CornerRadius() -> (CornerRadius:double)
   ) -> (then)
   > SetFontMaterials(Font Materials=Font Materials on Unhover())
  ]}
 SetCompactUIStyle {nodes [
   SetCompactUIStyle() {is=title K2Node_FunctionEntry}
   > Branch(Condition=IsUsingCompactUI())
   > Set CornerRadius(CornerRadius:double=Get CompactUICornerRadius() -> (CompactUICornerRadius:double)) -> (Output_Get:double)
  ]}
 SetFontMaterials {nodes [
   n1 = SetFontMaterials!() -> (then, Font Materials:Object Reference) {is=title K2Node_FunctionEntry}
   n2 = Break Slate Font Info!(SlateFontInfo:"Slate Font Info Structure (by ref)"=Get Font() -> (Font:Slate Font Info Structure)) -> (
    FontObject:Object Reference
    FontMaterial:Object Reference
    OutlineSettings:Font Outline Settings Structure
    TypefaceFontName:Name
    Size:float=~
    LetterSpacing:int
    SkewAmount:float=~
    bForceMonospaced:bool=~
    bMaterialIsStencil:bool=~
    MonospacedWidth:float=~
   ) {is=title K2Node_BreakStruct}
   > SetFont!(
    execute
    self:WBP UIKit Button Generic Base Object Reference
    In Font_FontObject:Object Reference=@n2
    In Font_FontMaterial:Object Reference=@n1
    In Font_OutlineSettings:Font Outline Settings Structure=@n2.OutlineSettings
    In Font_TypefaceFontName:Name=@n2.TypefaceFontName
    In Font_Size:float=@n2.Size
    In Font_LetterSpacing:int=@n2.LetterSpacing
    In Font_SkewAmount:float=@n2.SkewAmount
    In Font_bForceMonospaced:bool=@n2.bForceMonospaced
    In Font_bMaterialIsStencil:bool=@n2.bMaterialIsStencil
    In Font_MonospacedWidth:float=@n2.MonospacedWidth
   ) -> (then)
  ]}
 UpdateButtonType {nodes [
   n1 = UpdateButtonType!() -> (then, ButtonType:E_UI_Social_Button_Type Enum=NewEnumerator0) {is=title K2Node_FunctionEntry}
   > Set ButtonType(ButtonType:E_UI_Social_Button_Type Enum=@n1) -> (Output_Get:E_UI_Social_Button_Type Enum=NewEnumerator0)
   > InitializeButtonVisuals() {is=title K2Node_CallFunction}
  ]}
]
variables [
 ButtonType {type=E_UI_Social_Button_Type, default=Default}
 CompactUICornerRadius {type=number, default=20}
 CornerRadius {type=number, default=15}
 RadiusInPixels {type=boolean, default=true}
]
widgets {
 info {
  inheritedWidgetCount=0
  namedSlotCount=1
  parentClass=/FortUILibrary/Systems/ModularButton/WBP_UIKit_ButtonGeneric_Base.WBP_UIKit_ButtonGeneric_Base_C
  rootWidgetClass=null
  widgetCount=1
 }
 widgets [ButtonBackground {widgetClassPath=/FortniteCoreUI/Social/Buttons/Blocks/WBP_Social_Block_Background_Button, bIsVariable=true, widget=ButtonBackground}]
}
